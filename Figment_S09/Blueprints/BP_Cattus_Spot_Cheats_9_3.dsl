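; Debug cheat Blueprint for the Cattus spot (BP_Cattus_Spot_9_3), with console events to jump to viewpoints and trigger spot events along its spline.
; Teleporting picks the debug viewpoint nearest the spline position and runs BugItGo.
; Playback sends PlaySpotEvent by console command, with Nearby play and watch using the spline time closest to the view target.
; GoSpotNearby computes that time but teleports using time 0.
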(event Custom|GoSpot
  ; ke * GoSpot - Go to a view point near the latest spotting position
  (Default|TeleporttoWatchCattusPosition (Variables|/Figment_S09/Athena/Prototype/Blueprints/Cattus_9_3/BP_Cattus_Spot_9_3.BP_Cattus_Spot_9_3_C|GetPosition (Variables|Default|GetCattusSpot))))

(event Custom|GoSpotPos
  (params (Time :type Float%20%28double-precision%29))
  ; ke * Go SpotPos 0.5 - Go to a viewport near a specific spline position. Works on Editor
  (Default|TeleporttoWatchCattusPosition Time))

(event Custom|GoSpotNearby
  ; ke * GoSpotNearby - Go to the viewport near the closest position on the spline. Works in editor
  (CallFunction|GetClosestTimeonSpline)
  (Default|TeleporttoWatchCattusPosition 0.0))

(event Custom|PlaySpotNearby
  ; ke * PlaySpotNearby - Trigger the event at the closest point on the spline
  (bind _returnvalue (CallFunction|GetClosestTimeonSpline))
  (CallFunction|PlayEventatPosition _returnvalue))

(event Custom|WatchSpotPos
  (params (Time :type Float%20%28double-precision%29))
  ; ke * WatchSpotPos - Go to a viewport near a spline position, and trigger the event there
  (Default|TeleporttoWatchCattusPosition Time)
  (CallFunction|PlayEventatPosition Time))

(event Custom|WatchSpotNearby
  ; ke * WatchSpotNearby - Go to a viewport near a the closest position on the spline, and trigger the event
  (bind _returnvalue (CallFunction|GetClosestTimeonSpline))
  (CallFunction|WatchSpotPos _returnvalue))

(fn GetClosestTimeOnSpline
  (bind _returnvalue (Game|GetPlayerController 0))
  (bind _returnvalue_2 (Pawn|GetViewTarget _returnvalue))
  (Utilities|IsValid _returnvalue_2
    (:Is%20Valid
      (return (Spline|GetTimeatDistanceAlongSpline (Variables|/Figment_S09/Athena/Prototype/Blueprints/Cattus_9_3/BP_Cattus_Spot_9_3.BP_Cattus_Spot_9_3_C|GetSpline (Variables|Default|GetCattusSpot)) (Spline|GetDistanceAlongSplineatLocation (Variables|/Figment_S09/Athena/Prototype/Blueprints/Cattus_9_3/BP_Cattus_Spot_9_3.BP_Cattus_Spot_9_3_C|GetSpline (Variables|Default|GetCattusSpot)) (Transformation|GetActorLocation _returnvalue_2) "World"))))))

(fn PlayEventAtPosition
  (params (InDouble :type Float%20%28double-precision%29))
  (bind _returnvalue (Utilities|String|ToString(Float) InDouble))
  (bind _returnvalue_2 (Utilities|String|Append "cheat ke * PlaySpotEvent " _returnvalue " 20 10000"))
  (Development|ExecuteConsoleCommand _returnvalue_2)
  (Development|PrintString (Utilities|String|Append "Play Event Cmd: " _returnvalue_2)))

(fn TeleportToWatchCattusPosition
  (params (Time :type Float%20%28double-precision%29))
  (bind _localdebugtransform (LocalVariables|GetlocalDebugTransform))
  (bind (_location _rotation) (Math|Transform|BreakTransform _localdebugtransform))
  (bind (_location_x _location_y _location_z) (Math|Vector|BreakVector _location))
  (bind (_rotation_roll _rotation_pitch _rotation_yaw) (Math|Rotator|BreakRotator _rotation))
  (bind _returnvalue (Utilities|String|ToString(Float) _location_x))
  (bind _returnvalue_2 (Utilities|String|ToString(Float) _location_y))
  (bind _returnvalue_3 (Utilities|String|ToString(Float) _location_z))
  (bind _returnvalue_4 (Utilities|String|ToString(Float) _rotation_pitch))
  (bind _returnvalue_5 (Utilities|String|ToString(Float) _rotation_yaw))
  (bind _returnvalue_6 (Utilities|String|ToString(Float) _rotation_roll))
  (bind _returnvalue_7 (Utilities|String|Append "BugItGo " _returnvalue " " _returnvalue_2 " " _returnvalue_3 " " _returnvalue_4 " " _returnvalue_5 " " _returnvalue_6 " "))
  (LocalVariables|SetlocalCattusPosition (Spline|GetTransformatTime (Variables|/Figment_S09/Athena/Prototype/Blueprints/Cattus_9_3/BP_Cattus_Spot_9_3.BP_Cattus_Spot_9_3_C|GetSpline (Variables|Default|GetCattusSpot)) Time "World" true false))
  (LocalVariables|SetlocalDebugTransform (Utilities|Array|Get(acopy) (Variables|Default|GetDebugViewPoints)))
  (for _array_element (Variables|Default|GetDebugViewPoints)
    (if (< (Math|Vector|DistanceSquared(Vector) (.location _array_element) (.location (LocalVariables|GetlocalCattusPosition))) (Math|Vector|DistanceSquared(Vector) (.location (LocalVariables|GetlocalDebugTransform)) (.location (LocalVariables|GetlocalCattusPosition))))
      (LocalVariables|SetlocalDebugTransform _array_element)))
  (Development|ExecuteConsoleCommand _returnvalue_7)
  (Development|PrintString (Utilities|String|Append "Teleport Cmd: " _returnvalue_7)))

(fn ConstructionScript)
